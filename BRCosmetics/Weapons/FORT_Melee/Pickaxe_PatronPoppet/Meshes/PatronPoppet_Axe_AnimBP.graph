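# Animation Blueprint for the PatronPoppet pickaxe that adds secondary physics motion to its three leaf bones.
# Two Anim Dynamics chains run leaves A, B and C in sequence on the input pose with angular springs and gravity overrides.
# bIsInFrontEnd selects the World-space chain when true, otherwise a more damped RootRelative chain.
# Each update sets Gravity Override through a deprecated helper, but the AnimGraph does not read that variable.
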AnimBlueprint /BRCosmetics/Weapons/FORT_Melee/Pickaxe_PatronPoppet/Meshes/PatronPoppet_Axe_AnimBP
target_skeleton=&/PatronPoppet_Axe_Skeleton
graphs [
 AnimGraph {nodes [
   n1 = Anim Dynamics - Bone: dyn_leaf_A!(ComponentPose=Local To Component(LocalPose=Input Pose() {name=InPose_0}), bAlphaBoolEnabled:bool=True, ExternalForce="0,0,0") -> (Pose) {
    alphaInputType=Bool
    angularBiasOverride=1
    angularDampingOverride=0.9
    angularSpringConstant=1000
    bAngularSpring=true
    bOverrideAngularDamping=true
    bUseGravityOverride=true
    bUsePlanarLimit=false
    simulationSpace=World
    alphaBoolBlend {blendInTime=0.2, blendOption=Custom, blendOutTime=0.4}
    boundBone {boneName=dyn_leaf_A}
    componentAppliedLinearAccClamp=<400 400 400>
    componentLinearAccScale=<0.4 0.4 0.4>
    componentLinearVelScale=<0.5 0.5 0.5>
    gravityOverride {z=-979}
    physicsBodyDefinitions [{
      collisionType=CustomSphere
      sphereCollisionRadius=2.5
      boundBone {boneName=dyn_leaf_A}
      boxExtents=<9 6 6>
      constraintSetup {angularLimitsMax=<5 30 5>, angularLimitsMin=<-5 -30 -5>, angularTarget=<0 0 0>}
      localJointOffset=<2 0 0>
     }]
   }
   n2 = Anim Dynamics - Bone: dyn_leaf_B!(ComponentPose=@n1, bAlphaBoolEnabled:bool=True, ExternalForce="0,0,0") -> (Pose) {
    alphaInputType=Bool
    angularBiasOverride=1
    angularDampingOverride=0.9
    angularSpringConstant=1000
    bAngularSpring=true
    bOverrideAngularDamping=true
    bUseGravityOverride=true
    bUsePlanarLimit=false
    simulationSpace=World
    alphaBoolBlend {blendInTime=0.2, blendOption=Custom, blendOutTime=0.4}
    boundBone {boneName=dyn_leaf_B}
    componentAppliedLinearAccClamp=<3500 3500 3500>
    componentLinearAccScale=<0.4 0.4 0.4>
    componentLinearVelScale=<0.5 0.5 0.5>
    gravityOverride {z=-300}
    physicsBodyDefinitions [{
      collisionType=CustomSphere
      sphereCollisionRadius=2.5
      boundBone {boneName=dyn_leaf_B}
      boxExtents=<6 6 6>
      constraintSetup {angularLimitsMax=<8 30 8>, angularLimitsMin=<-8 -30 -8>, angularTarget=<0 0 0>}
      localJointOffset=<1 0 0>
     }]
   }
   n3 = Anim Dynamics - Bone: dyn_leaf_C!(ComponentPose=@n2, bAlphaBoolEnabled:bool=True, ExternalForce="0,0,0") -> (Pose) {
    alphaInputType=Bool
    angularBiasOverride=1
    angularDampingOverride=0.9
    angularSpringConstant=1000
    bAngularSpring=true
    bOverrideAngularDamping=true
    bUseGravityOverride=true
    bUsePlanarLimit=false
    simulationSpace=World
    alphaBoolBlend {blendInTime=0.2, blendOption=Custom, blendOutTime=0.4}
    boundBone {boneName=dyn_leaf_C}
    componentAppliedLinearAccClamp=<3500 3500 3500>
    componentLinearAccScale=<0.4 0.4 0.4>
    componentLinearVelScale=<0.5 0.5 0.5>
    gravityOverride {z=-200}
    physicsBodyDefinitions [{
      collisionType=CustomSphere
      sphereCollisionRadius=2.5
      boundBone {boneName=dyn_leaf_C}
      boxExtents=<4 4 4>
      constraintSetup {angularLimitsMax=<8 30 8>, angularLimitsMin=<-8 -30 -8>, angularTarget=<0 0 0>}
      localJointOffset=<2 0 0>
     }]
   }
   AnimGraphNode_SubInput_0 = Input Pose()
   n4 = Anim Dynamics - Bone: dyn_leaf_A(ComponentPose=Local To Component(LocalPose=@AnimGraphNode_SubInput_0), bAlphaBoolEnabled=True, ExternalForce="0,0,0") {
    alphaInputType=Bool
    angularBiasOverride=1
    angularDampingOverride=0.98
    angularSpringConstant=1000
    bAngularSpring=true
    bOverrideAngularDamping=true
    bUseGravityOverride=true
    bUsePlanarLimit=false
    simulationSpace=RootRelative
    alphaBoolBlend {blendInTime=0.2, blendOption=Custom, blendOutTime=0.4}
    boundBone {boneName=dyn_leaf_A}
    componentAppliedLinearAccClamp=<400 400 400>
    componentLinearAccScale=<0.2 0.2 0.2>
    componentLinearVelScale=<0.3 0.3 0.3>
    gravityOverride {z=-350}
    physicsBodyDefinitions [{
      collisionType=CustomSphere
      sphereCollisionRadius=2.5
      boundBone {boneName=dyn_leaf_A}
      boxExtents=<9 6 6>
      constraintSetup {angularLimitsMax=<5 30 5>, angularLimitsMin=<-5 -30 -5>, angularTarget=<0 0 0>}
      localJointOffset=<1 0 0>
     }]
   }
   n5 = Anim Dynamics - Bone: dyn_leaf_B(ComponentPose=@n4, bAlphaBoolEnabled=True, ExternalForce="0,0,0") {
    alphaInputType=Bool
    angularBiasOverride=1
    angularDampingOverride=0.98
    angularSpringConstant=1000
    bAngularSpring=true
    bOverrideAngularDamping=true
    bUseGravityOverride=true
    bUsePlanarLimit=false
    simulationSpace=RootRelative
    alphaBoolBlend {blendInTime=0.2, blendOption=Custom, blendOutTime=0.4}
    boundBone {boneName=dyn_leaf_B}
    componentAppliedLinearAccClamp=<500 500 500>
    componentLinearAccScale=<0.2 0.2 0.2>
    componentLinearVelScale=<0.3 0.3 0.3>
    gravityOverride {z=-370}
    physicsBodyDefinitions [{
      collisionType=CustomSphere
      sphereCollisionRadius=2.5
      boundBone {boneName=dyn_leaf_B}
      boxExtents=<8 5 5>
      constraintSetup {angularLimitsMax=<5 30 5>, angularLimitsMin=<-5 -30 -5>, angularTarget=<0 0 0>}
      localJointOffset=<1 0 0>
     }]
   }
   n6 = Anim Dynamics - Bone: dyn_leaf_C(ComponentPose=@n5, bAlphaBoolEnabled=True, ExternalForce="0,0,0") {
    alphaInputType=Bool
    angularBiasOverride=1
    angularDampingOverride=0.98
    angularSpringConstant=1000
    bAngularSpring=true
    bOverrideAngularDamping=true
    bUseGravityOverride=true
    bUsePlanarLimit=false
    simulationSpace=RootRelative
    alphaBoolBlend {blendInTime=0.2, blendOption=Custom, blendOutTime=0.4}
    boundBone {boneName=dyn_leaf_C}
    componentAppliedLinearAccClamp=<500 500 500>
    componentLinearAccScale=<0.2 0.2 0.2>
    componentLinearVelScale=<0.3 0.3 0.3>
    gravityOverride {z=-400}
    physicsBodyDefinitions [{
      collisionType=CustomSphere
      sphereCollisionRadius=2.5
      boundBone {boneName=dyn_leaf_C}
      boxExtents=<8 5 5>
      constraintSetup {angularLimitsMax=<5 30 5>, angularLimitsMin=<-5 -30 -5>, angularTarget=<0 0 0>}
      localJointOffset=<1 0 0>
     }]
   }
   Output Pose(Result=Blend Poses by bool(BlendPose_0=Component To Local(ComponentPose=@n3), BlendPose_1=Component To Local(ComponentPose=@n6), bActiveValue=Get bIsInFrontEnd()) {blendType=Linear}) {layerGroup=None}
  ]}
 EventGraph {nodes [
   Event BlueprintUpdateAnimation()
   > n1 = Gravity Override State Function DEPRECATED(CharacterPartAnimInstance=Self-Reference(), GameTimeInSeconds=GetGameTimeInSeconds(), GravityOverrideParams=Get Gravity Override Params())
   > Set Gravity Override(Gravity Override=@n1)
   TryGetPawnOwner()
  ]}
]
variables [
 Gravity Override {type=Vector}
 Gravity Override Params {type=GravityOverrideParamsStruct}
]
